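# UI master material for cycle button arrows, drawing a tree-arrow icon texture on a translucent surface.
# Normalized UI UVs have U scaled by the widget's width-to-height pixel ratio, and the static switch bIsRightFacing chooses between them and a horizontally mirrored copy.
# The texture alpha is multiplied by an Alpha parameter, alongside an RGB1 color parameter.
# A Hover scalar is declared but feeds no node in the graph.
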Material /FortUILibrary/Foundation/Buttons/Materials/Arrows/M_CycleButton_Arrow_Master
graph {
 properties {
  bUsedWithStaticMesh=false
  blendMode=BLEND_Translucent
  materialDomain=MD_UI
  pixelDepthOffsetMode=PDOM_Legacy
  usedShadingModels=Default Lit
 }
 nodes [
  n1 = MaterialFunctionCall() {materialFunction=&GetUserInterfaceUV}
  n2 = MaterialFunctionCall(Float2=@n1.Pixel Size) {materialFunction=&BreakOutFloat2Components}
  n3 = Max(A=@n2.G)
  n4 = Multiply(A=AppendVector(A=Divide(A=@n2.R, B=@n3), B=Constant() {r=1}), B=@n1.Normalized UV)
  n5 = MaterialFunctionCall(Float2=@n4) {materialFunction=&BreakOutFloat2Components}
  n6 = MaterialFunctionCall(X=OneMinus(None=@n5.R), Y=@n5.G) {materialFunction=&MakeFloat2}
  Multiply(
   A=TextureSample(UVs=StaticSwitchParameter(True=@n4, False=@n6.Result) {parameterName=bIsRightFacing}) {texture=/Game/UI/SocialPanel/Art/Textures/T_UI_Icon_TreeArrow_64}.A
   B=ScalarParameter() {defaultValue=1, parameterName=Alpha}
  )
  Reroute(None=VectorParameter() {parameterName=RGB1, defaultValue=<rgba 1 1 1 0>}.RGB)
  ScalarParameter() {parameterName=Hover}
 ]
}
