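# Translucent UI material for a button, named as a dropdown arrow, that draws a filled, outlined box with a center icon.
# UV distance from center builds the outline and a procedural plus or minus symbol, which IsArrow blends with a CenterIcon texture, horizontally mirrored by FlipArrow.
# The resulting masks layer separate fill, outline and icon parameters for color and alpha.
# A separate rotator-and-SmoothStep procedural shape branch does not feed these blends.
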Material /FortUILibrary/Foundation/Buttons/Materials/Arrows/M_DropDownArrow
graph {
 properties {
  bUsedWithStaticMesh=false
  blendMode=BLEND_Translucent
  materialDomain=MD_UI
  pixelDepthOffsetMode=PDOM_Legacy
  usedShadingModels=Default Lit
 }
 nodes [
  n1 = Reroute()
  n2 = OneMinus(None=Add(A=@n1, B=Constant2Vector() {g=0.5, r=0.3125}) {constB=-0.5})
  n3 = ScalarParameter() {defaultValue=0.125, parameterName=PrimaryAngle}
  n4 = Constant()
  n5 = ScalarParameter() {defaultValue=0.001, parameterName=SSMargin, sliderMax=1}
  n6 = MaterialFunctionCall(Alpha=ComponentMask(None=Rotator(Coordinate=@n2, Time=@n3) {centerX=0, centerY=0, speed=6.2832}) {r=true}, Min=@n4, Max=@n5) {materialFunction=&SmoothStep}
  n7 = OneMinus(None=Add(A=@n1, B=Constant2Vector() {g=0.5, r=0.4688}) {constB=-0.5})
  n8 = Rotator(Coordinate=@n7, Time=@n3) {centerX=0, centerY=0, speed=6.2832}
  n9 = Constant() {r=1}
  n10 = Subtract(A=@n9, B=@n5)
  n11 = MaterialFunctionCall(Alpha=OneMinus(None=ComponentMask(None=@n8) {r=true}), Min=@n10, Max=@n9) {materialFunction=&SmoothStep}
  n12 = ScalarParameter() {defaultValue=0.125, parameterName=PrimaryAngle}
  n13 = Rotator(Coordinate=Add(A=@n1, B=Constant2Vector() {r=-0.6125}) {constB=-0.5}, Time=Multiply(A=@n12) {constB=-1}) {centerX=0, centerY=0, speed=6.2832}
  n14 = Rotator(Coordinate=Add(A=@n1, B=Constant2Vector() {r=0.375}) {constB=-0.5}, Time=@n12) {centerX=0, centerY=0, speed=6.2832}
  n15 = MaterialFunctionCall(
   Alpha=DotProduct(A=ComponentMask(None=@n13) {r=true}, B=ComponentMask(None=@n14) {r=true})
   Min=Constant()
   Max=ScalarParameter() {defaultValue=0.001, parameterName=SSMargin, sliderMax=1}
  ) {materialFunction=&SmoothStep}
  n16 = Multiply(A=ScalarParameter() {defaultValue=0.125, parameterName=PrimaryAngle}) {constB=-1}
  n17 = Rotator(Coordinate=@n7, Time=@n16) {centerX=0, centerY=0, speed=6.2832}
  n18 = MaterialFunctionCall(Alpha=OneMinus(None=ComponentMask(None=@n17) {r=true}), Min=@n10, Max=@n9) {materialFunction=&SmoothStep}
  n19 = MaterialFunctionCall(Alpha=ComponentMask(None=Rotator(Coordinate=@n2, Time=@n16) {centerX=0, centerY=0, speed=6.2832}) {r=true}, Min=@n4, Max=@n5) {materialFunction=&SmoothStep}
  n20 = Add(A=@n1) {constB=-0.5}
  LinearInterpolate(
   A=DotProduct(A=DotProduct(A=@n6.Result, B=@n11.Result), B=@n15.Result)
   B=DotProduct(A=DotProduct(A=@n18.Result, B=@n19.Result), B=@n15.Result)
   Alpha=Ceil(None=ComponentMask(None=@n20) {g=true})
  )
  n21 = TextureCoordinate()
  n22 = Subtract(A=Abs(None=Add(A=@n21) {constB=-0.5}), B=Subtract(B=ScalarParameter() {defaultValue=0.0625, parameterName=OutlineThickness}) {constA=0.5})
  n23 = Clamp(None=MaterialFunctionCall(Vector 2=Ceil(None=@n22)) {materialFunction=/Engine/Functions/Engine_MaterialFunctions02/Utility/VectorLength}.V2 Length)
  n24 = Abs(None=Add(A=@n21) {constB=-0.5})
  n25 = Constant() {r=0.0625}
  n26 = Add(A=@n25, B=ScalarParameter() {defaultValue=0.001, parameterName=SSMargin, sliderMax=1})
  n27 = MaterialFunctionCall(Alpha=ComponentMask(None=@n24) {g=true}, Min=@n25, Max=@n26) {materialFunction=&SmoothStep}
  n28 = ComponentMask(None=@n24) {r=true}
  n29 = LinearInterpolate(
   A=@n27.Result
   B=DotProduct(
    A=MaterialFunctionCall(Alpha=@n28, Min=@n25, Max=@n26) {materialFunction=&SmoothStep}.Result
    B=@n27.Result
   )
   Alpha=ScalarParameter() {defaultValue=1, parameterName=IsPlusSymbol, sliderMax=1}
  )
  n30 = Subtract(A=Abs(None=Add(A=@n21) {constB=-0.5}), B=Subtract(B=Constant() {r=0.25}) {constA=0.5})
  n31 = Clamp(None=MaterialFunctionCall(Vector 2=Ceil(None=@n30)) {materialFunction=&VectorLength}.V2 Length)
  n32 = AppendVector(A=OneMinus(None=ComponentMask(None=@n21) {r=true}), B=ComponentMask(None=@n21) {g=true})
  n33 = ScalarParameter() {parameterName=FlipArrow}
  n34 = LinearInterpolate(
   A=DotProduct(A=OneMinus(None=@n29), B=OneMinus(None=@n31))
   B=TextureSampleParameter2D(UVs=LinearInterpolate(A=@n21, B=@n32, Alpha=@n33)) {parameterName=CenterIcon, texture=/Game/UI/Foundation/Textures/Buttons/T_CycleArrow}.A
   Alpha=ScalarParameter() {defaultValue=1, parameterName=IsArrow, sliderMax=1}
  )
  LinearInterpolate(
   A=LinearInterpolate(
    A=VectorParameter() {parameterName=FillRGB, defaultValue=<rgba 1 0.1823 0.3149 1>}.RGB
    B=VectorParameter() {parameterName=OutlineRGB, defaultValue=<rgba 0.1302 0.1699 1 1>}.RGB
    Alpha=@n23
   )
   B=VectorParameter() {parameterName=IconRGB, defaultValue=<rgba 0.9845 1 0.2634 1>}.RGB
   Alpha=@n34
  )
  LinearInterpolate(
   A=LinearInterpolate(
    A=ScalarParameter() {defaultValue=0.6, parameterName=FillAlpha}
    B=ScalarParameter() {defaultValue=0.9, parameterName=OutlineAlpha}
    Alpha=@n23
   )
   B=ScalarParameter() {defaultValue=1, parameterName=IconAlpha}
   Alpha=@n34
  )
 ]
}
