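# Translucent UI material that procedurally draws a chevron page arrow, with FlipDirection switching which way it points.
# FlipDirection blends the UV offset and rotation angles that feed pairs of mirrored Rotators.
# SmoothStep masks on those rotated coordinates form the outer edge, cut out the interior by Thickness, and trim the arm ends.
# SSMargin softens the edges, and the alpha of RGB1 scales the resulting mask.
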Material /FortUILibrary/Foundation/Buttons/Materials/Arrows/M_PageArrow_Chevron
graph {
 properties {
  bUsedWithStaticMesh=false
  blendMode=BLEND_Translucent
  materialDomain=MD_UI
  pixelDepthOffsetMode=PDOM_Legacy
  usedShadingModels=Default Lit
 }
 nodes [
  n1 = VectorParameter() {parameterName=RGB1, defaultValue=<rgba 1 0 0 1>}
  n2 = Constant() {r=0.175}
  n3 = Constant() {r=0.325}
  n4 = ScalarParameter() {parameterName=FlipDirection, sliderMax=1}
  n5 = Constant2Vector() {g=0.5}
  n6 = Constant2Vector() {g=0.5, r=1}
  n7 = ScalarParameter() {parameterName=FlipDirection, sliderMax=1}
  n8 = OneMinus(None=Add(A=TextureCoordinate(), B=LinearInterpolate(A=@n5, B=@n6, Alpha=@n7) {constAlpha=0}) {constB=-0.5})
  n9 = ComponentMask(None=Rotator(Coordinate=@n8, Time=LinearInterpolate(A=@n2, B=@n3, Alpha=@n4) {constAlpha=0}) {centerX=0, centerY=0, speed=6.2832}) {r=true}
  n10 = ScalarParameter() {defaultValue=0.125, parameterName=Thickness}
  n11 = ScalarParameter() {defaultValue=0.01, parameterName=SSMargin, sliderMax=1}
  n12 = Add(A=@n10, B=@n11)
  n13 = Multiply(A=@n2) {constB=-1}
  n14 = Multiply(A=@n3) {constB=-1}
  n15 = ComponentMask(None=Rotator(Coordinate=@n8, Time=LinearInterpolate(A=@n13, B=@n14, Alpha=@n4) {constAlpha=0}) {centerX=0, centerY=0, speed=6.2832}) {r=true}
  n16 = DotProduct(
   A=MaterialFunctionCall(Alpha=@n9, Min=@n10, Max=@n12) {materialFunction=&SmoothStep}.Result
   B=MaterialFunctionCall(Alpha=@n15, Min=@n10, Max=@n12) {materialFunction=&SmoothStep}.Result
  )
  n17 = Constant()
  n18 = Add(A=@n17, B=@n11)
  n19 = MaterialFunctionCall(Alpha=@n9, Min=@n17, Max=@n18) {materialFunction=&SmoothStep}
  n20 = MaterialFunctionCall(Alpha=@n15, Min=@n17, Max=@n18) {materialFunction=&SmoothStep}
  n21 = Constant() {r=0.775}
  n22 = Constant() {r=0.725}
  n23 = LinearInterpolate(A=@n21, B=@n22, Alpha=@n4) {constAlpha=0}
  n24 = ComponentMask(None=Add(A=Rotator(Coordinate=@n8, Time=@n23) {centerX=0, centerY=0, speed=6.2832}) {constB=-0.5}) {r=true}
  n25 = LinearInterpolate(A=Multiply(A=@n21) {constB=-1}, B=Multiply(A=@n22) {constB=-1}, Alpha=@n4) {constAlpha=0}
  n26 = ComponentMask(None=Add(A=Rotator(Coordinate=@n8, Time=@n25) {centerX=0, centerY=0, speed=6.2832}) {constB=-0.5}) {r=true}
  LinearInterpolate(
   A=LinearInterpolate(B=LinearInterpolate(A=@n1.A, Alpha=@n16) {constB=0}, Alpha=DotProduct(A=@n19.Result, B=@n20.Result)) {constB=0}
   Alpha=Add(
    A=MaterialFunctionCall(Alpha=@n24, Min=@n17, Max=@n11) {materialFunction=&SmoothStep}.Result
    B=MaterialFunctionCall(Alpha=@n26, Min=@n17, Max=@n11) {materialFunction=&SmoothStep}.Result
   )
  ) {constB=0}
 ]
}
