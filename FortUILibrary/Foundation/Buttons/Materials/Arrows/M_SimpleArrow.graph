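# UI-domain translucent material that procedurally draws a simple arrow shape for button visuals, without textures.
# Two copies of the UVs are rotated in opposite 45-degree directions, thresholded by SmoothStep for soft edges, and joined at the vertical midpoint.
# sInverseDirection flips both the rotations and the mask to reverse the arrow.
# vRGB supplies color, sAlpha scales the mask's opacity, and sSSMargin sets edge softness.
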Material /FortUILibrary/Foundation/Buttons/Materials/Arrows/M_SimpleArrow
graph {
 properties {
  bUsedWithStaticMesh=false
  blendMode=BLEND_Translucent
  materialDomain=MD_UI
  pixelDepthOffsetMode=PDOM_Legacy
  usedShadingModels=Default Lit
 }
 nodes [
  n1 = OneMinus(None=TextureCoordinate() {coordinateIndex=4, vTiling=2})
  n2 = ScalarParameter() {defaultValue=1, parameterName=sInverseDirection, sliderMax=1}
  n3 = LinearInterpolate(A=Constant() {r=-0.125}, B=Constant() {r=0.125}, Alpha=@n2)
  n4 = Rotator(Coordinate=Add(A=@n1) {constB=-0.5}, Time=@n3) {centerX=0, centerY=0, speed=6.2832}
  n5 = Constant()
  n6 = ScalarParameter() {defaultValue=-0.01, parameterName=sSSMargin}
  n7 = Rotator(Coordinate=Add(A=@n1) {constB=0.5}, Time=OneMinus(None=@n3)) {centerX=1, centerY=0, speed=6.2832}
  n8 = Constant() {r=1}
  n9 = Add(A=TextureCoordinate() {coordinateIndex=4}) {constB=-0.5}
  n10 = LinearInterpolate(
   A=MaterialFunctionCall(Alpha=ComponentMask(None=@n4) {r=true}, Min=Add(A=@n5, B=@n6), Max=@n5) {materialFunction=&SmoothStep}.Result
   B=MaterialFunctionCall(Alpha=ComponentMask(None=@n7) {r=true}, Min=Add(A=@n8, B=@n6), Max=@n8) {materialFunction=&SmoothStep}.Result
   Alpha=Ceil(None=ComponentMask(None=@n9) {g=true})
  )
  n11 = OneMinus(None=@n10)
  LinearInterpolate(B=ScalarParameter() {defaultValue=0.5, parameterName=sAlpha, sliderMax=1}, Alpha=OneMinus(None=LinearInterpolate(A=@n10, B=@n11, Alpha=@n2)))
  VectorParameter() {parameterName=vRGB, defaultValue=<rgba 1 0 0 1>}
 ]
}
